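# Sound cue for the Clyde dance grenade explosion, routed to the large explosives sound class with dance grenade explosion attenuation.
# A quality level node selects between two mixers, one using the standard close explosion wave and the other its "opt" variant.
# Each mixer layers that wave with a quieter, delayed, randomly chosen generic bomb explosion pitched randomly at or above normal.
# A concurrency override prevents new instances once the limit is reached.
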SoundCue /Clyde/Audio/Items/LemonCart_Weapons/Clyde_DanceGrenade_Explosion_Athena_Cue
root=QualityLevel_0
nodes [
 {id=Delay_0, children [Modulator_2], properties {delayMax=0.4, delayMin=0.4}}
 {id=Delay_1, children [Modulator_3], properties {delayMax=0.4, delayMin=0.4}}
 {id=Mixer_0, children [Modulator_0,Delay_1], properties {inputVolume [1,0.2]}}
 {id=Mixer_1, children [Modulator_1,Delay_0], properties {inputVolume [1,0.2]}}
 {id=Modulator_0, children [WavePlayer_5], properties {pitchMax=1, pitchMin=1}}
 {id=Modulator_1, children [WavePlayer_8], properties {pitchMax=1, pitchMin=1}}
 {id=Modulator_2, children [Random_0], properties {pitchMax=1.6, pitchMin=1}}
 {id=Modulator_3, children [Random_1], properties {pitchMax=1.6, pitchMin=1}}
 {id=QualityLevel_0, children [Mixer_1,Mixer_1,Mixer_0,Mixer_0]}
 {id=Random_0, children [WavePlayer_3,WavePlayer_4], properties {weights [1,1]}}
 {id=Random_1, children [WavePlayer_0,WavePlayer_1], properties {weights [1,1]}}
 {id=WavePlayer_0, properties {soundWaveAssetPtr=/Game/Sounds/Fort_GamePlay_Sounds/Explosives/bomb_explode_stereo_01}}
 {id=WavePlayer_1, properties {soundWaveAssetPtr=/Game/Sounds/Fort_GamePlay_Sounds/Explosives/bomb_explode_stereo_02}}
 {id=WavePlayer_3, properties {soundWaveAssetPtr=&bomb_explode_stereo_01}}
 {id=WavePlayer_4, properties {soundWaveAssetPtr=&bomb_explode_stereo_02}}
 {id=WavePlayer_5, properties {soundWaveAssetPtr=&/Clyde_sfx_dancegrenade_explosion_close_opt}}
 {id=WavePlayer_8, properties {soundWaveAssetPtr=&/Clyde_sfx_dancegreande_explosion_close}}
]
properties {
 attenuationSettings=/Game/Sounds/Attenuation/Fort_Explosion_DanceGrenade_Att
 bOverrideConcurrency=true
 maxDistance=8000
 soundClassObject=/Game/Sounds/FortSoundClassesAndModes/Fort_Explosives_Large_PSM
 volumeMultiplier=0.5307
 concurrencyOverrides {resolutionRule=PreventNew}
}
