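# Sound cue for Clyde's Dance Grenade music, classed and concurrency-limited as first-person emote music.
# A fixed short delay leads into a licensed-content switcher, whose first two inputs share the Athena dance grenade loop and whose third is a Clyde variant.
# Only the Clyde wave player is set to loop.
# Concurrency is capped at one, and output also feeds a music audio-analysis submix.
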SoundCue /Clyde/Audio/Items/LemonCart_Weapons/Clyde_DanceGrenade_Music_Cue
root=Delay_0
nodes [
 {id=Delay_0, children [FortSoundNodeLicensedContentSwitcher_0], properties {delayMax=0.5, delayMin=0.5}}
 {id=FortSoundNodeLicensedContentSwitcher_0, class=FortSoundNodeLicensedContentSwitcher, children [WavePlayer_1,WavePlayer_1,WavePlayer_0]}
 {id=WavePlayer_0, properties {bLooping=true, soundWaveAssetPtr=&/MX_IP_Clyde_Athena_DanceGrenade_Loop}}
 {id=WavePlayer_1, properties {soundWaveAssetPtr=/Game/Athena/Sounds/DanceGrenade/music_dancegrenade_loop}}
]
properties {
 maxDistance=2097152
 soundClassObject=/Game/Sounds/FortSoundClassesAndModes/Fort_Music_Emote_1P_PSM
 concurrencyOverrides {maxCount=1}
 concurrencySet [/Game/Sounds/Concurrency_Rules/Athena_EmoteMusic_1P_Conc]
 soundSubmixSends [{minSendLevel=0, soundSubmix=&Fortnite_AudioAnalysis_Music}]
}
