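# Pickup sound cue for the Clyde LemonCart AR weapon, using loot-collect sound class and weapon-pickup attenuation.
# A quality-level node picks one of two mixers, each layering a rifle pickup wave at double volume with a shared licensed-content switcher.
# The switcher plays a Midas stinger only in its third slot; the others are empty.
# Concurrency limits stack pickups, and the cue feeds a boss VO envelope-follower bus.
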SoundCue /Clyde/Audio/Items/LemonCart_Weapons/Clyde_LemonCart_AR_Pickup_Cue
root=QualityLevel_0
nodes [
 {id=FortSoundNodeLicensedContentSwitcher_0, class=FortSoundNodeLicensedContentSwitcher, children [null,null,WavePlayer_0]}
 {id=Mixer_0, children [WavePlayer_5, FortSoundNodeLicensedContentSwitcher_0], properties {inputVolume [2,1]}}
 {id=Mixer_1, children [WavePlayer_4, FortSoundNodeLicensedContentSwitcher_0], properties {inputVolume [2,1]}}
 {id=QualityLevel_0, children [Mixer_1,Mixer_1,Mixer_0,Mixer_0]}
 {id=WavePlayer_0, properties {soundWaveAssetPtr=&/MX_IP_Midas_Clyde_Pickup_Stinger}}
 {id=WavePlayer_4, properties {soundWaveAssetPtr=/Clyde/Audio/Items/BerryTart_Weapons/Clyde_Rifle_Pickup}}
 {id=WavePlayer_5, properties {soundWaveAssetPtr=&Clyde_Rifle_Pickup}}
]
properties {
 attenuationSettings=&Fort_Weapon_Pickup_Att
 maxDistance=3500
 soundClassObject=&Fort_Loot_Collect
 volumeMultiplier=0.8
 busSends [{audioBus=/Game/Sounds/Modulation/EnvelopeFollowers/BR_Boss/AB_BR_Boss_VO_EnvFollow, minSendLevel=0, sendLevel=4, soundSourceBus=null}]
 concurrencyOverrides {maxCount=1, resolutionRule=PreventNew}
 concurrencySet [&PickupMaximum_Oldest, /Clyde/Audio/Items/Clyde_Weapon_Pickup_Concurrency]
}
