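# Sound cue for the looping third-person Dance Grenade music in Clyde, played as spatialized emote music.
# A fixed short delay feeds a licensed content switcher, which uses the stock Athena loop in its first two slots and a Clyde-specific loop in the third.
# It keeps playing when silent, limits simultaneous instances through emote music concurrency, and sends to a music audio-analysis submix.
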SoundCue /Clyde/Audio/Items/LemonCart_Weapons/SC_Clyde_DanceGrenade_Music_3P_Cue
root=Delay_0
nodes [
 {id=Delay_0, children [FortSoundNodeLicensedContentSwitcher_0], properties {delayMax=0.5, delayMin=0.5}}
 {id=FortSoundNodeLicensedContentSwitcher_0, class=FortSoundNodeLicensedContentSwitcher, children [WavePlayer_2,WavePlayer_2,WavePlayer_1]}
 {id=WavePlayer_1, properties {bLooping=true, soundWaveAssetPtr=&/MX_IP_Clyde_Athena_DanceGrenade_Loop}}
 {id=WavePlayer_2, properties {bLooping=true, soundWaveAssetPtr=/Game/Athena/Sounds/DanceGrenade/music_dancegrenade_loop}}
]
properties {
 attenuationSettings=/Game/Sounds/Attenuation/Fort_Emotes_DanceGrenadeMusic_Att
 maxDistance=2350
 soundClassObject=/Game/Sounds/FortSoundClassesAndModes/Fort_Music_Emote_3P
 virtualizationMode=PlayWhenSilent
 concurrencyOverrides {maxCount=2, volumeScale=0.5}
 concurrencySet [/Game/Sounds/Concurrency_Rules/Athena_EmoteMusic_Conc]
 soundSubmixSends [{minSendLevel=0, soundSubmix=&Fortnite_AudioAnalysis_Music}]
}
